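# Lit material for a wet rock wall in the Apollo Mole set, with condensation and running water.
# Diffuse alpha plus an offset masks tiled condensation and panning second-UV drips, which feed base color, metallic, specular and roughness.
# Line streaks broken up by a tile pattern vary specular and roughness.
# Material layer functions blend drip and flattened condensation normals over the rock normal with angle correction, then drips reduce roughness.
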Material /Game/Environments/Apollo/Sets/Mole/Materials/M_Mole_RockWall_Wet
graph {
 properties {bUseMaterialAttributes=true, pixelDepthOffsetMode=PDOM_Legacy, usedShadingModels=Default Lit}
 nodes [
  n1 = Constant() {r=0.5}
  n2 = TextureSample() {texture=/Game/Environments/Apollo/Sets/Mole/Textures/T_Mole_Rockwall_A_D}
  n3 = Add(A=ScalarParameter() {defaultValue=0.1, parameterName=Occluded Area Mask}, B=@n2.A)
  n4 = TextureSample(UVs=TextureCoordinate() {uTiling=3, vTiling=1.5}) {samplerType=SAMPLERTYPE_LinearColor, texture=/Game/Environments/Apollo/Sets/Mole/Textures/T_Mole_RockWall_Condensation_N}
  n5 = LinearInterpolate(B=Multiply(A=@n3, B=@n4.A), Alpha=ScalarParameter() {defaultValue=0.2, parameterName=Condensation Amount})
  n6 = TextureSample(UVs=Panner(
    Coordinate=TextureCoordinate() {coordinateIndex=1, uTiling=2}
    Speed=VectorParameter() {parameterName=Speed, defaultValue=<rgba 0 -0.02 0 1>}.RGB
   )) {samplerType=SAMPLERTYPE_LinearColor, texture=/Game/Environments/Apollo/Sets/Mole/Textures/T_Mole_RockWall_Drips_N}
  n7 = Multiply(A=@n3, B=@n6.A)
  n8 = Multiply(A=ScalarParameter() {defaultValue=0.1, parameterName=rain_color}, B=@n7)
  n9 = TextureSample() {samplerType=SAMPLERTYPE_Masks, texture=/Game/Environments/Apollo/Sets/Mole/Textures/T_Mole_Rockwall_A_S}
  n10 = TextureSample() {texture=/Game/Effects/Fort_Effects/Textures/T_Tile_Pattern_01}
  n11 = TextureSample(UVs=TextureCoordinate()) {samplerType=SAMPLERTYPE_Masks, texture=/Game/Environments/Sets/Vulcano/Textures/T_LineStreaks_M}
  n12 = Multiply(A=@n3, B=@n11.R)
  n13 = Multiply(A=@n3, B=@n6.A)
  n14 = Add(A=@n5, B=@n9.R)
  n15 = LinearInterpolate(A=@n9.B, B=Constant(), Alpha=@n5)
  n16 = LinearInterpolate(A=ScalarParameter() {defaultValue=0.2, parameterName=SecondaryStreaksRoughness}, B=@n15, Alpha=OneMinus(None=@n11.R))
  n17 = Multiply(A=@n10.R, B=@n3)
  n18 = MakeMaterialAttributes!(
   BaseColor=Add(A=Multiply(A=@n1, B=@n5), B=Add(A=@n8, B=@n2.RGB))
   Metallic=Add(A=@n5, B=Add(A=@n7, B=@n9.G))
   Specular=Add(A=Multiply(A=@n10.R, B=@n12), B=Add(A=@n13, B=@n14))
   Roughness=LinearInterpolate(
    A=LinearInterpolate(A=@n15, B=@n16, Alpha=@n17)
    B=Constant()
    Alpha=ScalarParameter() {defaultValue=0.5, parameterName=OverallRoughnessBias}
   )
   Anisotropy
   EmissiveColor
   Opacity
   OpacityMask
   Normal=TextureSample() {samplerType=SAMPLERTYPE_Normal, texture=/Game/Environments/Apollo/Sets/Mole/Textures/T_Mole_Rockwall_A_N}.RGB
   Tangent
   WorldPositionOffset
   SubsurfaceColor
   ClearCoat
   ClearCoatRoughness
   AmbientOcclusion
   Refraction
   CustomizedUVs_0
   CustomizedUVs_1
   CustomizedUVs_2
   CustomizedUVs_3
   CustomizedUVs_4
   CustomizedUVs_5
   CustomizedUVs_6
   CustomizedUVs_7
   PixelDepthOffset
   ShadingModel
   Displacement
  )
  n19 = ScalarParameter() {defaultValue=0.5, parameterName=Condensation Flatness}
  MaterialFunctionCall!(
   Base Material=MaterialFunctionCall!(
    Material=MaterialFunctionCall(Material=@n18, Additional Normal=@n6.RGB, Mask=@n13) {desc=Streamingwater, materialFunction=/Engine/Functions/MaterialLayerFunctions/MatLayerBlend_BlendAngleCorrectedNormals}.Blended Material
    Additional Normal=MaterialFunctionCall(Normal=@n4.RGB, Flatness=@n19) {materialFunction=&FlattenNormal}.Result
    Mask=@n4.A
   ) {desc=CondensationNormal, materialFunction=&MatLayerBlend_BlendAngleCorrectedNormals}.Blended Material
   RoughnessMultiplier=OneMinus(None=@n7)
  ) {materialFunction=/Engine/Functions/MaterialLayerFunctions/MatLayerBlend_ModulateRoughness}
 ]
}
